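; Hanging lamp prop Blueprint with a flipbook flame mesh and a point light, both color-configurable through variables.
; The construction script gives the flame a dynamic material instance and sets its hue-shift parameter from the normalized HSV hue difference between FX Color and a reference orange.
; It also applies a separate Point Light Color variable to the light.
; BeginPlay, overlap and Tick events are empty.
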(event EventBeginPlay)

(event Collision|EventActorBeginOverlap
  (params (OtherActor :type Actor%20Object%20Reference)))

(event EventTick
  (params (DeltaSeconds :type Float%20%28single-precision%29)))

(fn ConstructionScript
  (|CallParentFunction)
  (bind _returnvalue (Rendering|Material|CreateDynamicMaterialInstance (Variables|Default|GetSM_Flame_Flipbook) 0 (Rendering|Material|GetMaterial (Variables|Default|GetSM_Flame_Flipbook) 0) "None" :!class /Script/Engine.PrimitiveComponent))
  (Rendering|Material|SetScalarParameterValue _returnvalue "Hue Shift Percentage (S)" (/ (Utilities|Struct|BreakLinearColor (- (Math|Color|RGBtoHSV(LinearColor) (Variables|Default|GetFX%20Color)) (Math|Color|RGBtoHSV(LinearColor) "(R=1.000000,G=0.250000,B=0.000000,A=1.000000)"))) 3.6e+02) :!class /Script/Engine.MaterialInstanceDynamic)
  (Rendering|Components|Light|SetLightColor (Variables|Default|GetPointLight) (Variables|Default|GetPointLight%20Color) true))
